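# Physics asset for the Summer Heroes glider skeletal mesh in the VinderTech GliderChute cosmetics.
# A single body on the Handles bone uses one large, offset and rotated box as a coarse collision and bounds volume.
# It has no constraints, so it defines no articulated or jointed physics between parts of the glider.
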PhysicsAsset /BRCosmetics/Gadgets/Assets/VinderTech_GliderChute/Glider_Summer_Heroes/Meshes/SK_Glider_Summer_Heroes_PhysicsAsset
bodies {Handles [{shapeType=Box, extentX=350.7394, extentY=219.4678, extentZ=354.9946, center=<-11.3392 14.5601 83.4533>, rotation=<rot 0 50.0261 -69.7672>}]}
constraints []
